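# Niagara system for the Chequered Flag vehicle boost, trailing a ribbon and sparks behind skeletal-mesh boost ports.
# Each emitter spreads particles across filtered sockets and takes its colour from painted-variant overrides.
# The ribbon adds age-driven sine waves through a RibbonWPO position, with width, brightness and material parameters animated by curves.
# Two sprite ember emitters use three spawn groups written to visibility tags: one streams straight back, the other swirls via vortex force.
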NiagaraSystem /VehicleCosmetics/Boosts/Boost_ChequeredFlag/FX/NS_Boost_ChequeredFlag_MP
dependencies {
 usedDataInterfaces [
  &NiagaraDataInterfaceCurve
  &NiagaraDataInterfaceSpriteRendererInfo
  &NiagaraDataInterfaceSkeletalMesh
 ]
 usedDynamicInputs []
 usedModules [
  &self:SystemSpawnScript.SystemScriptSource.SystemScriptGraph.NiagaraNodeAssignment_0.SetVariables_#1
  &SystemState
 ]
}
emitters {
 Embers002 {
  emitterUpdateScript [
   &EmitterState(Life Cycle Mode=::NewEnumerator1, Loop Duration=2, UseLoopDelay=false)
   &VC_FilteredSocketCount()
   &SpawnRate(SpawnRate=40, Use Spawn Probability=false)
   &SpawnRate(SpawnGroup=1, SpawnRate=80, Use Spawn Probability=false) {module=SpawnRate001}
   &SpawnRate(SpawnGroup=2, SpawnRate=80, Use Spawn Probability=false) {module=SpawnRate002}
  ]
  particleSpawnScript [
   &VC_PaintedVariantsOverride(
    Crimson=<rgba 0.013 0 0 5>
    DefaultColor=<rgba 0.8174 0.1717 0.0121 3>
    ForestGreen=<rgba 0 0.0573 0.0029 1>
    Lime=<rgba 0.0273 0.036 0 10>
    Orange=<rgba 0.0417 0.0107 0 3>
   )
   &InitializeParticle(
    Color Mode=::NewEnumerator0
    Lifetime=&MaskFloat(
     Mask Source=0
     Mask Parameter=?
     Mask=0
     Masked Float=&RandomRangeFloat(Minimum=0.2, Maximum=0.4)
     Passthrough Float=&RandomRangeFloat(Minimum=0.2, Maximum=0.45)
    )
    Sprite Rotation Mode=::NewEnumerator2
    Sprite Size Mode=::NewEnumerator1
    Uniform Sprite Size Max=100
    Uniform Sprite Size Min=70
    UsePositionOffset=true
    Write Material Random=true
   )
   &PartitionParticles(Number Of Partitions=$Emitter.PortCount)
   &SkeletalMeshLocation(
    Bone / Socket Index=$Particles.PartitionParticles.Partition
    Bone / Socket Sampling Mode=::NewEnumerator5
    MaskSpawnGroup=false
    Mesh Sampling Type=::NewEnumerator6
    Sampled Position Offset=<-20 0 0>
    Skeletal Mesh=$System.SkeletalMesh
   )
   /SetVariables(
    Particles.SparkCircSize:Vector2f=<4 4>
    Particles.SparkLineSize:Vector2f=&RandomRangeVector2D(Minimum=<1.5 12>, Maximum=<2 35>)
   ) {isSetParameters=true, module=SetVariables_#2}
   &ShapeLocation(MaskSpawnGroup=false, Override Seed=false, Sphere Radius=36, UseOwnerScale=true) {module=ShapeLocation001}
   &AddVelocity(Velocity=&MaskVector(
     Mask Source=0
     Mask Parameter=?
     Mask=0
     Masked Vector=<-2400 0 0>
     Passthrough Vector=&MaskVector(Mask Source=0, Mask Parameter=?, Mask=1, Masked Vector=<-2000 0 0>, Passthrough Vector=<-1500 0 0>)
    ))
   /SetVariables(Particles.VisibilityTag=$Emitter.SpawnGroup) {isSetParameters=true, module=SetVariables_#3}
  ]
  particleUpdateScript [
   &ParticleState()
   &SpriteRotationRate(Rotation Rate=0.5)
   &SolveForcesAndVelocity(Clamp Velocity=false, Limit Acceleration=false)
   &ScaleColor(
    Scale Alpha=&FloatFromCurve(FloatCurve=DataInterface() {propertyValues {Curve {keys [{tangentMode=RCTM_User, time=0, value=1}, {tangentMode=RCTM_User, time=1, value=0}]}}})
    Scale RGB=&VectorFromFloat(Value=80)
    ScaleA=true
    ScaleRGB=true
   ) {module=ScaleColor001}
   &SubUVAnimation(
    End Frame Range Override=3
    Output Alternate Renderer Binding=true
    "Sprite Renderer "=DataInterface() {propertyValues {SpriteRenderer=&self:Embers002_0.NiagaraSpriteRendererProperties_2}}
    SubUV Animation Mode:ENiagara_SubUVLookupModeV2=::NewEnumerator2
    UseEndFrame=true
    UseStartFrame=true
   ) {module=SubUVAnimation001}
   /SetVariables(Particles.SparkColor:LinearColor=&MultiplyLinearColor(
     Linear Color=$Particles.Color
     Scale Factor=&MakeVector4(
      X=10
      Y=10
      Z=10
      W=&FloatFromCurve(FloatCurve=DataInterface() {propertyValues {Curve {keys [<at 0 1>,<at 1 0>]}}})
     )
    )) {isSetParameters=true, module=SetVariables_#4}
   &DynamicMaterialParameters(
    Index 1 Param 1=&FloatFromCurve(FloatCurve=DataInterface() {propertyValues {Curve {defaultValue=1, keys [{arriveTangent=0.1759, time=0.2, value=0.5}, {arriveTangent=5.0898, tangentMode=RCTM_User, time=1, value=1}]}}})
    Index 1 Param 2=0.25
    Module.Index0Float0Write=true
    Module.Index0Float1Write=true
    Module.Index0Float2Write=true
    Module.Index0Float3Write=true
    Module.Index1Float0Write=true
    Module.Index1Float1Write=true
    Module.Index1Float2Write=true
    Module.Index1Float3Write=true
    Write Parameter Index 1=::NewEnumerator1
   )
  ]
  rendererClasses [&NiagaraSpriteRendererProperties]
  renderers [
   {rendererClass=&NiagaraSpriteRendererProperties}
   {rendererClass=&NiagaraSpriteRendererProperties, rendererIndex=1}
   {rendererClass=&NiagaraSpriteRendererProperties, rendererIndex=2}
  ]
 }
 Embers003 {
  emitterUpdateScript [
   &EmitterState()
   &VC_FilteredSocketCount()
   &SpawnRate(SpawnRate=30, Use Spawn Probability=false)
   &SpawnRate(SpawnGroup=1, SpawnRate=50, Use Spawn Probability=false) {module=SpawnRate001}
   &SpawnRate(SpawnGroup=2, SpawnRate=50, Use Spawn Probability=false) {module=SpawnRate002}
  ]
  particleSpawnScript [
   &VC_PaintedVariantsOverride(
    Crimson=<rgba 0.013 0 0 5>
    DefaultColor=<rgba 0.8174 0.1717 0.0121 3>
    ForestGreen=<rgba 0 0.0573 0.0029 1>
    Lime=<rgba 0.0273 0.036 0 10>
    Orange=<rgba 0.0417 0.0107 0 3>
   )
   &InitializeParticle(
    Color Mode=::NewEnumerator0
    Lifetime=&MaskFloat(
     Mask Source=0
     Mask Parameter=?
     Mask=0
     Masked Float=&RandomRangeFloat(Minimum=0.1, Maximum=0.4)
     Passthrough Float=&RandomRangeFloat(Minimum=0.2, Maximum=0.5)
    )
    Sprite Rotation Mode=::NewEnumerator2
    Sprite Size Mode=::NewEnumerator1
    Uniform Sprite Size Max=100
    Uniform Sprite Size Min=70
    UsePositionOffset=true
    Write Material Random=true
   )
   &PartitionParticles(Number Of Partitions=$Emitter.PortCount)
   &SkeletalMeshLocation(
    Bone / Socket Index=$Particles.PartitionParticles.Partition
    Bone / Socket Sampling Mode=::NewEnumerator5
    MaskSpawnGroup=false
    Mesh Sampling Type=::NewEnumerator6
    Sampled Position Offset=<-20 0 0>
    Skeletal Mesh=$System.SkeletalMesh
   )
   /SetVariables(
    Particles.SparkCircSize:Vector2f=<4 4>
    Particles.SparkLineSize:Vector2f=&RandomRangeVector2D(Minimum=<1 12>, Maximum=<2 35>)
   ) {isSetParameters=true, module=SetVariables_#2}
   &ShapeLocation(MaskSpawnGroup=false, Override Seed=false, Sphere Radius=36, UseOwnerScale=true) {module=ShapeLocation001}
   &AddVelocity(Velocity=&MaskVector(
     Mask Source=0
     Mask Parameter=?
     Mask=0
     Masked Vector=&RandomRangeVector(Minimum=<-1400 -100 -40>, Maximum=<-700 100 100>)
     Passthrough Vector=&MaskVector(
      Mask Source=0
      Mask Parameter=?
      Mask=1
      Masked Vector=&RandomRangeVector(Minimum=<-1000 -100 -50>, Maximum=<-500 100 100>)
      Passthrough Vector=&RandomRangeVector(Minimum=<-700 -100 -50>, Maximum=<-350 100 100>)
     )
    ))
   /SetVariables(Particles.VisibilityTag=$Emitter.SpawnGroup) {isSetParameters=true, module=SetVariables_#3}
  ]
  particleUpdateScript [
   &ParticleState()
   &SpriteRotationRate()
   &VortexForce(Origin Pull Amount=500, UseOriginOffset=false, Vortex Axis=<-10 0 0>, Vortex Force Amount=1500)
   &SolveForcesAndVelocity(Clamp Velocity=false, Limit Acceleration=false)
   &ScaleColor(
    Scale Alpha=&FloatFromCurve(FloatCurve=DataInterface() {propertyValues {Curve {keys [{tangentMode=RCTM_User, time=0, value=1}, {tangentMode=RCTM_User, time=1, value=0}]}}})
    Scale RGB=&VectorFromFloat(Value=10)
    ScaleA=true
    ScaleRGB=true
   ) {module=ScaleColor001}
   &SubUVAnimation(
    End Frame Range Override=3
    Output Alternate Renderer Binding=true
    "Sprite Renderer "=DataInterface() {propertyValues {SpriteRenderer=&self:Embers003_0.NiagaraSpriteRendererProperties_2}}
    SubUV Animation Mode:ENiagara_SubUVLookupModeV2=::NewEnumerator2
    UseEndFrame=true
    UseStartFrame=true
   ) {module=SubUVAnimation001}
   /SetVariables(Particles.SparkColor:LinearColor=&MultiplyLinearColor(
     Linear Color=$Particles.Color
     Scale Factor=&MakeVector4(
      X=250
      Y=250
      Z=250
      W=&FloatFromCurve(FloatCurve=DataInterface() {propertyValues {Curve {keys [<at 0 1>,<at 1 0>]}}})
     )
    )) {isSetParameters=true, module=SetVariables_#4}
   &DynamicMaterialParameters(
    Index 1 Param 1=&FloatFromCurve(FloatCurve=DataInterface() {propertyValues {Curve {defaultValue=1, keys [{arriveTangent=0.1759, time=0.2, value=0.5}, {arriveTangent=5.0898, tangentMode=RCTM_User, time=1, value=1}]}}})
    Index 1 Param 2=0.25
    Module.Index0Float0Write=true
    Module.Index0Float1Write=true
    Module.Index0Float2Write=true
    Module.Index0Float3Write=true
    Module.Index1Float0Write=true
    Module.Index1Float1Write=true
    Module.Index1Float2Write=true
    Module.Index1Float3Write=true
    Write Parameter Index 1=::NewEnumerator1
   )
  ]
  rendererClasses [&NiagaraSpriteRendererProperties]
  renderers [
   {rendererClass=&NiagaraSpriteRendererProperties}
   {rendererClass=&NiagaraSpriteRendererProperties, rendererIndex=1}
   {rendererClass=&NiagaraSpriteRendererProperties, rendererIndex=2}
  ]
 }
 NE_VC_Boost_Ribbon_Parent {
  emitterUpdateScript [
   &EmitterState()
   &VC_FilteredSocketCount()
   &VC_RateQualityHandler(High Spawn Rate=60, Low Spawn Rate=60, Medium Spawn Rate=60, Use Custom Rate?=true)
   &VC_SpawnRibbon_v1(Number of Ribbons=$Emitter.PortCount, SpawnRate=$Emitter.Spawn Rate)
  ]
  particleSpawnScript [
   &VC_PaintedVariantsOverride(
    Cobalt=<rgba 0.0297 0 2 1>
    Crimson=<rgba 1.2 0 0.0325 1>
    DefaultColor=<rgba 1 0.3059 0.0385 1>
    ForestGreen=<rgba 0 0.4896 0.0247 2>
    Lime=<rgba 0.5346 0.9219 0 1>
    Orange=<rgba 0.4948 0.0819 0 1>
    Pink=<rgba 0.8125 0.2624 0.6125 1>
    Purple=<rgba 0.4104 0 1 1>
    R1 Cobalt:LinearColor=<rgba 0.0497 0.099 0.8281 1>
    R1 ForestGreen:LinearColor=<rgba 0 0.2708 0.0137 1>
    R1 Lime:LinearColor=<rgba 0.2976 0.5521 0 1>
    R1 Orange:LinearColor=<rgba 0.3698 0.0797 0 1>
    R1 Pink:LinearColor=<rgba 0.5208 0.1492 0.3857 1>
    R1 Purple:LinearColor=<rgba 0.3146 0.0406 0.7083 1>
    R1 Saffron:LinearColor=<rgba 0.4844 0.4484 0.0202 1>
    R1 SkyBlue:LinearColor=<rgba 0 0.198 0.5365 1>
    SkyBlue=<rgba 0 0.3394 0.6875 1>
    Use R1 Painted Overrides=true
   )
   &VC_InitializeRibbon_v1(Color=$Particles.Color, Enable Skeletal Position=true, Ribbon Width=280)
   &AddVelocity(Velocity=<-4000 0 0>) {module=AddVelocity001}
  ]
  particleUpdateScript [
   &ParticleState()
   &SolveForcesAndVelocity(Clamp Velocity=false, Limit Acceleration=false)
   &VC_UpdateRibbon_v1(Enable Skeletal Position=true)
   &ScaleRibbonWidth(Ribbon Width Scale=&FloatFromCurve(FloatCurve=DataInterface() {propertyValues {Curve {keys [<at 0 0.5>,<at 0.2 1.3>]}}}))
   /SetVariables(Particles.RibbonWPO:NiagaraPosition=&AddVectorToPosition(
     Position=$Particles.Position
     Vector=&TransformVector(
      Vector=&MakeVector(
       X=0
       Y=&Sine(
        Normalized Angle=&Add_Float(
         A=&Multiply_Float(A=$Particles.NormalizedAge, B=-1.2)
         B=&Multiply_Float(A=$Emitter.LoopedAge, B=1)
        )
        Period=0.7
        Scale=&FloatFromCurve(
         FloatCurve=DataInterface() {propertyValues {Curve {defaultValue=1, keys [{tangentMode=RCTM_User, time=0, value=0}, {tangentMode=RCTM_User, time=0.53, value=1.1}]}}}
         Scale Curve=10
        )
        Bias=0
       )
       Z=&Sine(
        Normalized Angle=&Add_Float(
         A=&Multiply_Float(A=$Particles.NormalizedAge, B=-1.3)
         B=&Multiply_Float(A=$Emitter.LoopedAge, B=1)
        )
        Period=0.8
        Scale=&FloatFromCurve(
         FloatCurve=DataInterface() {propertyValues {Curve {
            defaultValue=1
            keys [
             {tangentMode=RCTM_User, time=0.1, value=0.3}
             {arriveTangent=-0.2778, interpMode=RCIM_Cubic, leaveTangent=-3.8301, tangentMode=RCTM_Break, time=0.3, value=1.2}
             <at 0.5 1 RCIM_Cubic>
            ]
           }}}
         Scale Curve=9
        )
        Bias=0
       )
      )
      Destination Space=::Simulation
      Apply Scale=true
     )
    )) {isSetParameters=true}
   &DynamicMaterialParameters(
    Index 0 Param 1=&FloatFromCurve(FloatCurve=DataInterface() {propertyValues {Curve {keys [<at 0.1 1>,<at 0.4 0>,<at 1 0>]}}})
    Index 0 Param 2=0.1
    Index 1 Param 1=&FloatFromCurve(FloatCurve=DataInterface() {propertyValues {Curve {keys [<at 0 1>,<at 0.15 0>,<at 0.4 1>]}}})
    Index 1 Param 2=0.1
    Module.Index0Float0Write=true
    Module.Index0Float1Write=true
    Module.Index0Float2Write=true
    Module.Index0Float3Write=true
    Module.Index1Float0Write=true
    Module.Index1Float1Write=true
    Module.Index1Float2Write=true
    Module.Index1Float3Write=true
    Write Parameter Index 1=::NewEnumerator1
   )
   &ScaleColor(
    Scale RGB=&VectorFromFloat(Value=&FloatFromCurve(FloatCurve=DataInterface() {propertyValues {Curve {keys [<at 0.1 6 RCIM_Cubic>, {arriveTangent=-10.5748, leaveTangent=-1.1111, tangentMode=RCTM_User, time=0.4, value=1}, <at 1 1>]}}}))
    ScaleA=true
    ScaleRGB=true
   )
  ]
  rendererClasses [&NiagaraRibbonRendererProperties]
  renderers [
   {rendererClass=&NiagaraRibbonRendererProperties}
   {rendererClass=&NiagaraRibbonRendererProperties, rendererIndex=1}
  ]
 }
}
summary {
 systemName=NS_Boost_ChequeredFlag_MP
 emitters [
  {emitterName=NE_VC_Boost_Ribbon_Parent, rendererClasses [&NiagaraRibbonRendererProperties]}
  {emitterName=Embers002, rendererClasses [&NiagaraSpriteRendererProperties]}
  {emitterName=Embers003, rendererClasses [&NiagaraSpriteRendererProperties]}
 ]
 userVariables [
  User.bIsFrontendPreview {type=NiagaraBool}
  User.bIsPainted {type=NiagaraBool}
  User.ColorVariantID {type=NiagaraInt32}
  User.DTCarOffset {type=Vector3f, description="", defaultValue {struct=Vector3f, value=<0 0 0>}}
  User.PaintedColor {type=LinearColor, description="", defaultValue {struct=LinearColor, value=<rgba 1 1 1 1>}}
  User.PaintedScalar {type=NiagaraFloat, defaultValue=1, description=""}
  User.PortCount {type=NiagaraInt32, defaultValue=1}
 ]
}
